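# Sound cue for the close sound of the default Lobster item, classed as weapon foley.
# A Random node picks evenly between two close wave variations and feeds a Modulator that varies the pitch.
# It plays at low volume with the Lobster melee-equip attenuation asset.
# A concurrency override caps simultaneous instances at two, stopping the lowest priority.
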SoundCue /Galileo/Items/Lobster/Assets/Sounds/Default/Lobster_Default_Close_Cue
root=Modulator_0
nodes [
 {id=Modulator_0, children [Random_0], properties {pitchMax=1.1}}
 {id=Random_0, children [WavePlayer_0,WavePlayer_3], properties {bShouldExcludeFromBranchCulling=true, weights [1,1]}}
 {id=WavePlayer_0, properties {soundWaveAssetPtr=&/Lobster_Default_Close_02}}
 {id=WavePlayer_3, properties {soundWaveAssetPtr=&/Lobster_Default_Close_01}}
]
properties {
 attenuationSettings=/Game/Sounds/Attenuation/Fort_Weapon_Melee_Lobster_Equip_Att
 bOverrideConcurrency=true
 maxDistance=4000
 soundClassObject=&Fort_WeaponFoley
 volumeMultiplier=0.113
 concurrencyOverrides {maxCount=2, resolutionRule=StopLowestPriority}
}
