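# Melee deflect sound cue for the default Lobster weapon, mixing a deflect layer with a poise warning layer, both driven by the InvPoise parameter.
# The deflect layer picks randomly from ten waves on the first two quality levels, four on the others, with InvPoise modulating pitch.
# The warning layer crossfades in at high InvPoise, playing one of two pitch-modulated waves borrowed from the Kayak variant.
# A concurrency override caps simultaneous instances.
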SoundCue /Galileo/Items/Lobster/Assets/Sounds/Default/Lobster_Default_Deflect_Cue
root=Mixer_0
nodes [
 {id=Mixer_0, children [ModulatorContinuous_1, ParamCrossFade_1], properties {inputVolume [1,1]}}
 {
  id=ModulatorContinuous_1
  children [QualityLevel_0]
  properties {pitchModulationParams {default=0, maxOutput=1.2, minOutput=0.8, parameterName=InvPoise}}
 }
 {
  id=ModulatorContinuous_2
  children [Random_0]
  properties {pitchModulationParams {default=0, maxOutput=1.5, minOutput=0.9, parameterName=InvPoise}}
 }
 {
  id=ParamCrossFade_1
  children [ModulatorContinuous_2]
  properties {
   paramName=InvPoise
   crossFadeInput [{fadeInDistanceEnd=1, fadeInDistanceStart=0.7, fadeOutDistanceEnd=100, fadeOutDistanceStart=1, volume=1.2}]
  }
 }
 {id=QualityLevel_0, children [Random_2,Random_2,Random_1,Random_1]}
 {id=Random_0, children [WavePlayer_1,WavePlayer_2], properties {bShouldExcludeFromBranchCulling=true, weights [1,1]}}
 {
  id=Random_1
  children [WavePlayer_0,WavePlayer_7,WavePlayer_8,WavePlayer_9]
  properties {bShouldExcludeFromBranchCulling=true, weights [1,1,1,1]}
 }
 {
  id=Random_2
  children [
   WavePlayer_3
   WavePlayer_4
   WavePlayer_5
   WavePlayer_6
   WavePlayer_16
   WavePlayer_17
   WavePlayer_18
   WavePlayer_19
   WavePlayer_20
   WavePlayer_21
  ]
  properties {bShouldExcludeFromBranchCulling=true, weights [1,1,1,1,1,1,1,1,1,1]}
 }
 {id=WavePlayer_0, properties {soundWaveAssetPtr=&/Lobster_Default_Deflect_06}}
 {id=WavePlayer_1, properties {soundWaveAssetPtr=/Galileo/Items/Lobster/Assets/Sounds/Kayak/Lobster_Kayak_Poise_Warning_01}}
 {id=WavePlayer_16, properties {soundWaveAssetPtr=&/Lobster_Default_Deflect_04}}
 {id=WavePlayer_17, properties {soundWaveAssetPtr=&/Lobster_Default_Deflect_05}}
 {id=WavePlayer_18, properties {soundWaveAssetPtr=&Lobster_Default_Deflect_06}}
 {id=WavePlayer_19, properties {soundWaveAssetPtr=&/Lobster_Default_Deflect_07}}
 {id=WavePlayer_2, properties {soundWaveAssetPtr=/Galileo/Items/Lobster/Assets/Sounds/Kayak/Lobster_Kayak_Poise_Warning_02}}
 {id=WavePlayer_20, properties {soundWaveAssetPtr=&/Lobster_Default_Deflect_08}}
 {id=WavePlayer_21, properties {soundWaveAssetPtr=&/Lobster_Default_Deflect_09}}
 {id=WavePlayer_3, properties {soundWaveAssetPtr=&/Lobster_Default_Deflect_10}}
 {id=WavePlayer_4, properties {soundWaveAssetPtr=&/Lobster_Default_Deflect_01}}
 {id=WavePlayer_5, properties {soundWaveAssetPtr=&/Lobster_Default_Deflect_02}}
 {id=WavePlayer_6, properties {soundWaveAssetPtr=&/Lobster_Default_Deflect_03}}
 {id=WavePlayer_7, properties {soundWaveAssetPtr=&Lobster_Default_Deflect_07}}
 {id=WavePlayer_8, properties {soundWaveAssetPtr=&Lobster_Default_Deflect_08}}
 {id=WavePlayer_9, properties {soundWaveAssetPtr=&Lobster_Default_Deflect_09}}
]
properties {
 attenuationSettings=/Game/Sounds/Attenuation/Fort_Weapon_Melee_Lobster_Attack_Deflect_Stereo_Att
 bOverrideConcurrency=true
 maxDistance=12000
 soundClassObject=&Fort_WeaponMelee
 volumeMultiplier=0.6748
 concurrencyOverrides {maxCount=3, resolutionRule=StopLowestPriority}
}
