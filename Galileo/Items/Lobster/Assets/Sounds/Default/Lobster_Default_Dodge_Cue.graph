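# Sound cue for the dodge sound of the default Lobster item, played as a melee weapon sound.
# A Random node picks one of two dodge wave variants with equal weight, and a Modulator above it varies the pitch.
# It uses close-range stereo melee attenuation and a concurrency override that caps simultaneous instances, stopping the lowest priority one.
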SoundCue /Galileo/Items/Lobster/Assets/Sounds/Default/Lobster_Default_Dodge_Cue
root=Modulator_0
nodes [
 {id=Modulator_0, children [Random_0], properties {pitchMax=1.1}}
 {id=Random_0, children [WavePlayer_1,WavePlayer_2], properties {bShouldExcludeFromBranchCulling=true, weights [1,1]}}
 {id=WavePlayer_1, properties {soundWaveAssetPtr=&/Lobster_Default_Dodge_01}}
 {id=WavePlayer_2, properties {soundWaveAssetPtr=&/Lobster_Default_Dodge_02}}
]
properties {
 attenuationSettings=/Game/Sounds/Attenuation/Fort_Weapon_Melee_Attack_Close_Range_Stereo_Att
 bOverrideConcurrency=true
 maxDistance=10058.75
 soundClassObject=&Fort_WeaponMelee
 volumeMultiplier=0.731
 concurrencyOverrides {maxCount=2, resolutionRule=StopLowestPriority}
}
