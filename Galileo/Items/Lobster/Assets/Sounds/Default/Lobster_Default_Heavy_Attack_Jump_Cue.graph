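# Sound cue for the jump heavy attack of the default Lobster melee weapon, currently an empty shell with no root node or wave players.
# It routes to the melee weapon sound class with close-range stereo attack attenuation and a reduced volume.
# Two Lobster swing concurrency assets, named for owner and global limits, are assigned; the cue's concurrency override specifies stopping the lowest priority.
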SoundCue /Galileo/Items/Lobster/Assets/Sounds/Default/Lobster_Default_Heavy_Attack_Jump_Cue
root=null
nodes []
properties {
 attenuationSettings=/Game/Sounds/Attenuation/Fort_Weapon_Melee_Attack_Close_Range_Stereo_Att
 maxDistance=4000
 soundClassObject=&Fort_WeaponMelee
 volumeMultiplier=0.5623
 concurrencyOverrides {resolutionRule=StopLowestPriority}
 concurrencySet [/Galileo/Items/Lobster/Assets/Sounds/Lobster_Swings_OwnerLimit_Conc, /Galileo/Items/Lobster/Assets/Sounds/Lobster_Swings_GlobalLimit_Conc]
}
