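# Sound cue for the looping idle sound of the default Lobster melee weapon.
# A single looping wave player plays the Lobster default loop wave at reduced volume.
# It routes through the melee weapon sound class with a close-range idle attenuation asset.
# A concurrency override caps simultaneous instances, stopping the lowest priority when the limit is exceeded.
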SoundCue /Galileo/Items/Lobster/Assets/Sounds/Default/Lobster_Default_IdleLoop_Cue
root=WavePlayer_1
nodes [{id=WavePlayer_1, properties {bLooping=true, soundWaveAssetPtr=&/Lobster_Default_Loop_01}}]
properties {
 attenuationSettings=/Game/Sounds/Attenuation/Fort_Weapon_Melee_Lobster_Idle_Loop_Close_Att
 bOverrideConcurrency=true
 maxDistance=3000
 soundClassObject=&Fort_WeaponMelee
 volumeMultiplier=0.5623
 concurrencyOverrides {maxCount=4, resolutionRule=StopLowestPriority}
}
