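# Sound cue for the default impact sound of the Lobster item, routed to the melee weapon sound class.
# A Random node with equal weights picks one of six impact waves, and a Modulator above it adds slight pitch variation on top of a raised base pitch and lowered volume.
# The cue uses its own attenuation asset and overrides concurrency to allow two simultaneous instances, stopping the oldest.
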SoundCue /Galileo/Items/Lobster/Assets/Sounds/Default/Lobster_Default_Impact_Cue
root=Modulator_0
nodes [
 {id=Modulator_0, children [Random_1], properties {pitchMax=1.1}}
 {
  id=Random_1
  children [WavePlayer_0, WavePlayer_7, WavePlayer_8, WavePlayer_9, WavePlayer_10, WavePlayer_11]
  properties {bShouldExcludeFromBranchCulling=true, weights [1,1,1,1,1,1]}
 }
 {id=WavePlayer_0, properties {soundWaveAssetPtr=&/Lobster_Default_Impact_01}}
 {id=WavePlayer_10, properties {soundWaveAssetPtr=&/Lobster_Default_Impact_05}}
 {id=WavePlayer_11, properties {soundWaveAssetPtr=&/Lobster_Default_Impact_06}}
 {id=WavePlayer_7, properties {soundWaveAssetPtr=&/Lobster_Default_Impact_02}}
 {id=WavePlayer_8, properties {soundWaveAssetPtr=&/Lobster_Default_Impact_03}}
 {id=WavePlayer_9, properties {soundWaveAssetPtr=&/Lobster_Default_Impact_04}}
]
properties {
 attenuationSettings=&/Lobster_Default_Impact_Att
 bOverrideConcurrency=true
 maxDistance=6000
 pitchMultiplier=1.2
 soundClassObject=&Fort_WeaponMelee
 volumeMultiplier=0.3
 concurrencyOverrides {maxCount=2, resolutionRule=StopOldest}
}
