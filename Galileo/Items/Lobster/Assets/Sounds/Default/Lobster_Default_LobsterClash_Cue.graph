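# Sound cue for the default Lobster melee weapon's clash, with close-range melee attenuation and a two-voice concurrency cap.
# A mixer layers two branches: a random clash from five waves, and a random Kayak poise-warning sound from two waves.
# The InvPoise parameter modulates the pitch of both branches and crossfades the warning layer in only near the top of its 0–1 range, signaling poise state through sound.
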SoundCue /Galileo/Items/Lobster/Assets/Sounds/Default/Lobster_Default_LobsterClash_Cue
root=Mixer_0
nodes [
 {id=Mixer_0, children [ModulatorContinuous_1, ParamCrossFade_1], properties {inputVolume [1,1]}}
 {
  id=ModulatorContinuous_1
  children [Random_0]
  properties {pitchModulationParams {default=0, maxOutput=1.5, minOutput=0.8, parameterName=InvPoise}}
 }
 {
  id=ModulatorContinuous_2
  children [Random_1]
  properties {pitchModulationParams {default=0, maxOutput=1.5, minOutput=0.9, parameterName=InvPoise}}
 }
 {
  id=ParamCrossFade_1
  children [ModulatorContinuous_2]
  properties {
   paramName=InvPoise
   crossFadeInput [{fadeInDistanceEnd=1, fadeInDistanceStart=0.8, fadeOutDistanceEnd=100, fadeOutDistanceStart=1, volume=1.2}]
  }
 }
 {
  id=Random_0
  children [WavePlayer_1,WavePlayer_2,WavePlayer_3,WavePlayer_4,WavePlayer_5]
  properties {bShouldExcludeFromBranchCulling=true, weights [1,1,1,1,1]}
 }
 {id=Random_1, children [WavePlayer_0,WavePlayer_6], properties {bShouldExcludeFromBranchCulling=true, weights [1,1]}}
 {id=WavePlayer_0, properties {soundWaveAssetPtr=/Galileo/Items/Lobster/Assets/Sounds/Kayak/Lobster_Kayak_Poise_Warning_01}}
 {id=WavePlayer_1, properties {soundWaveAssetPtr=&/Lobster_Default_LobsterClash_01}}
 {id=WavePlayer_2, properties {soundWaveAssetPtr=&/Lobster_Default_LobsterClash_02}}
 {id=WavePlayer_3, properties {soundWaveAssetPtr=&/Lobster_Default_LobsterClash_03}}
 {id=WavePlayer_4, properties {soundWaveAssetPtr=&/Lobster_Default_LobsterClash_04}}
 {id=WavePlayer_5, properties {soundWaveAssetPtr=&/Lobster_Default_LobsterClash_05}}
 {id=WavePlayer_6, properties {soundWaveAssetPtr=/Galileo/Items/Lobster/Assets/Sounds/Kayak/Lobster_Kayak_Poise_Warning_02}}
]
properties {
 attenuationSettings=/Game/Sounds/Attenuation/Fort_Weapon_Melee_Attack_Close_Range_Stereo_Att
 bOverrideConcurrency=true
 maxDistance=10058.75
 soundClassObject=&Fort_WeaponMelee
 volumeMultiplier=0.5623
 concurrencyOverrides {maxCount=2, resolutionRule=StopLowestPriority}
}
