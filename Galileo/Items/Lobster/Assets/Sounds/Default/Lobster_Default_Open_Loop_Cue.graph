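# Sound cue for the looping open-state sound of the Lobster melee weapon's default variant.
# A delay node feeds a modulator with fixed reduced volume into a looping wave player, so the loop starts after a short, constant delay.
# It uses the Lobster equip attenuation and melee weapon sound class, with overridden concurrency that caps simultaneous instances by stopping the lowest priority.
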SoundCue /Galileo/Items/Lobster/Assets/Sounds/Default/Lobster_Default_Open_Loop_Cue
root=Delay_0
nodes [
 {id=Delay_0, children [Modulator_0], properties {delayMax=0.4, delayMin=0.4}}
 {id=Modulator_0, children [WavePlayer_9], properties {volumeMax=0.8, volumeMin=0.8}}
 {id=WavePlayer_9, properties {bLooping=true, soundWaveAssetPtr=&/Lobster_Default_Loop_01}}
]
properties {
 attenuationSettings=/Game/Sounds/Attenuation/Fort_Weapon_Melee_Lobster_Equip_Att
 bOverrideConcurrency=true
 maxDistance=4000
 soundClassObject=&Fort_WeaponMelee
 volumeMultiplier=0.56
 concurrencyOverrides {maxCount=3, resolutionRule=StopLowestPriority}
}
