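# Sound cue for picking up the default Lobster item in Galileo, reusing the shared Revolver_Pickup wave.
# A single wave player feeds a Modulator left at its default settings, with the cue's volume turned well down.
# It uses the weapon pickup attenuation and weapon foley sound class, and joins the PickupMaximum_Oldest concurrency set to limit overlapping pickup sounds.
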SoundCue /Galileo/Items/Lobster/Assets/Sounds/Default/Lobster_Default_Pickup_Cue
root=Modulator_0
nodes [
 {id=Modulator_0, children [WavePlayer_2]}
 {id=WavePlayer_2, properties {soundWaveAssetPtr=/Game/Sounds/Fort_GamePlay_Sounds/Loot/Revolver_Pickup}}
]
properties {
 attenuationSettings=&Fort_Weapon_Pickup_Att
 maxDistance=3500
 soundClassObject=&Fort_WeaponFoley
 volumeMultiplier=0.282
 concurrencyOverrides {maxCount=1, resolutionRule=StopLowestPriority}
 concurrencySet [&PickupMaximum_Oldest]
}
