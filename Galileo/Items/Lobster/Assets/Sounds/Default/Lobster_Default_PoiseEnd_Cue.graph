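# Sound cue for the poise-end sound of the Lobster item's default variant, classed as a melee weapon sound.
# A single wave player feeds a modulator that adds slight random pitch variation.
# It uses a close-range stereo melee attack attenuation and overrides concurrency to cap simultaneous instances, stopping the lowest priority.
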SoundCue /Galileo/Items/Lobster/Assets/Sounds/Default/Lobster_Default_PoiseEnd_Cue
root=Modulator_0
nodes [
 {id=Modulator_0, children [WavePlayer_0], properties {pitchMax=1.1}}
 {id=WavePlayer_0, properties {soundWaveAssetPtr=&/Lobster_Default_PoiseEnd_01}}
]
properties {
 attenuationSettings=/Game/Sounds/Attenuation/Fort_Weapon_Melee_Attack_Close_Range_Stereo_Att
 bOverrideConcurrency=true
 maxDistance=10058.75
 soundClassObject=&Fort_WeaponMelee
 volumeMultiplier=0.5623
 concurrencyOverrides {maxCount=2, resolutionRule=StopLowestPriority}
}
