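# Sound cue for the start of the Lobster item's poise sound in its default set, classed as melee weapon audio.
# A random node picks one of four equally weighted poise-start wave variations and is excluded from branch culling.
# It plays at reduced volume through a close-range mono melee attack attenuation.
# A concurrency override caps it at two simultaneous instances, stopping the lowest priority.
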SoundCue /Galileo/Items/Lobster/Assets/Sounds/Default/Lobster_Default_PoiseStart_Cue
root=Random_1
nodes [
 {
  id=Random_1
  children [WavePlayer_0,WavePlayer_3,WavePlayer_4,WavePlayer_5]
  properties {bShouldExcludeFromBranchCulling=true, weights [1,1,1,1]}
 }
 {id=WavePlayer_0, properties {soundWaveAssetPtr=&/Lobster_Default_Poise_Start_01}}
 {id=WavePlayer_3, properties {soundWaveAssetPtr=&/Lobster_Default_Poise_Start_02}}
 {id=WavePlayer_4, properties {soundWaveAssetPtr=&/Lobster_Default_Poise_Start_03}}
 {id=WavePlayer_5, properties {soundWaveAssetPtr=&/Lobster_Default_Poise_Start_04}}
]
properties {
 attenuationSettings=/Game/Sounds/Attenuation/Fort_Weapon_Melee_Attack_Close_Range_Mono_Att
 bOverrideConcurrency=true
 maxDistance=10058.75
 soundClassObject=&Fort_WeaponMelee
 volumeMultiplier=0.25
 concurrencyOverrides {maxCount=2, resolutionRule=StopLowestPriority}
}
