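# Sound cue for the quick-open sound of the default Lobster item, classed and attenuated as a melee weapon sound.
# One of four equally weighted open samples is picked at random, given slight volume variation, and played after a short fixed delay.
# A concurrency override caps simultaneous instances, stopping the lowest priority when the limit is reached.
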SoundCue /Galileo/Items/Lobster/Assets/Sounds/Default/Lobster_Default_QuickOpen_Cue
root=Delay_0
nodes [
 {id=Delay_0, children [Modulator_1], properties {delayMax=0.15, delayMin=0.15}}
 {id=Modulator_1, children [Random_1], properties {volumeMax=1, volumeMin=0.8}}
 {
  id=Random_1
  children [WavePlayer_5,WavePlayer_6,WavePlayer_7,WavePlayer_8]
  properties {bShouldExcludeFromBranchCulling=true, weights [1,1,1,1]}
 }
 {id=WavePlayer_5, properties {soundWaveAssetPtr=&/Lobster_Default_Open_01}}
 {id=WavePlayer_6, properties {soundWaveAssetPtr=&/Lobster_Default_Open_02}}
 {id=WavePlayer_7, properties {soundWaveAssetPtr=&/Lobster_Default_Open_03}}
 {id=WavePlayer_8, properties {soundWaveAssetPtr=&/Lobster_Default_Open_04}}
]
properties {
 attenuationSettings=/Game/Sounds/Attenuation/Fort_Weapon_Melee_Lobster_Equip_Att
 bOverrideConcurrency=true
 maxDistance=4000
 soundClassObject=&Fort_WeaponMelee
 volumeMultiplier=0.665
 concurrencyOverrides {maxCount=3, resolutionRule=StopLowestPriority}
}
