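# Sound cue for the fourth combo swing of the default Lobster melee weapon.
# A random node picks one of four equally weighted SwingFour wave variations.
# It plays in the melee weapon sound class with close-range stereo melee attack attenuation and reduced volume.
# Owner and global concurrency sets limit overlapping Lobster swings.
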SoundCue /Galileo/Items/Lobster/Assets/Sounds/Default/Lobster_Default_Swing_Combo_04_Cue
root=Random_0
nodes [
 {
  id=Random_0
  children [WavePlayer_0,WavePlayer_1,WavePlayer_2,WavePlayer_3]
  properties {bShouldExcludeFromBranchCulling=true, weights [1,1,1,1]}
 }
 {id=WavePlayer_0, properties {soundWaveAssetPtr=&/Lobster_Default_SwingFour_01}}
 {id=WavePlayer_1, properties {soundWaveAssetPtr=&/Lobster_Default_SwingFour_02}}
 {id=WavePlayer_2, properties {soundWaveAssetPtr=&/Lobster_Default_SwingFour_03}}
 {id=WavePlayer_3, properties {soundWaveAssetPtr=&/Lobster_Default_SwingFour_04}}
]
properties {
 attenuationSettings=/Game/Sounds/Attenuation/Fort_Weapon_Melee_Attack_Close_Range_Stereo_Att
 maxDistance=10058.75
 soundClassObject=&Fort_WeaponMelee
 volumeMultiplier=0.6748
 concurrencyOverrides {resolutionRule=StopLowestPriority}
 concurrencySet [/Galileo/Items/Lobster/Assets/Sounds/Lobster_Swings_OwnerLimit_Conc, /Galileo/Items/Lobster/Assets/Sounds/Lobster_Swings_GlobalLimit_Conc]
}
