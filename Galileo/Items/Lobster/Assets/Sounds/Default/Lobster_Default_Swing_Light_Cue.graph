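# Sound cue for the light swing of the default Lobster melee item.
# A Random node picks with equal weight among seven wave variations and is excluded from branch culling.
# It plays at reduced volume through the melee weapon sound class and a close-range stereo melee attenuation.
# Two shared Lobster swing concurrency assets, named for owner and global limits, govern simultaneous swings.
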SoundCue /Galileo/Items/Lobster/Assets/Sounds/Default/Lobster_Default_Swing_Light_Cue
root=Random_0
nodes [
 {
  id=Random_0
  children [WavePlayer_2,WavePlayer_3,WavePlayer_4,WavePlayer_5,WavePlayer_6,WavePlayer_7,WavePlayer_8]
  properties {bShouldExcludeFromBranchCulling=true, weights [1,1,1,1,1,1,1]}
 }
 {id=WavePlayer_2, properties {soundWaveAssetPtr=&/Lobster_Default_SwingLight_02}}
 {id=WavePlayer_3, properties {soundWaveAssetPtr=&/Lobster_Default_SwingLight_03}}
 {id=WavePlayer_4, properties {soundWaveAssetPtr=&/Lobster_Default_SwingLight_04}}
 {id=WavePlayer_5, properties {soundWaveAssetPtr=&/Lobster_Default_SwingLight_05}}
 {id=WavePlayer_6, properties {soundWaveAssetPtr=&/Lobster_Default_SwingLight_06}}
 {id=WavePlayer_7, properties {soundWaveAssetPtr=&/Lobster_Default_SwingLight_07}}
 {id=WavePlayer_8, properties {soundWaveAssetPtr=&/Lobster_Default_SwingLight_08}}
]
properties {
 attenuationSettings=/Game/Sounds/Attenuation/Fort_Weapon_Melee_Attack_Close_Range_Stereo_Att
 maxDistance=10058.75
 priority=0.5
 soundClassObject=&Fort_WeaponMelee
 volumeMultiplier=0.1406
 concurrencyOverrides {resolutionRule=StopLowestPriority}
 concurrencySet [/Galileo/Items/Lobster/Assets/Sounds/Lobster_Swings_OwnerLimit_Conc, /Galileo/Items/Lobster/Assets/Sounds/Lobster_Swings_GlobalLimit_Conc]
}
